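; Director condition asset for the Time of Day Manager scalability sequence, testing whether the game runs on Nintendo Switch (Erebus).
; Its Evaluate function returns the result of the TouchPlatform check for Switch or Switch Preview.
; It also writes that boolean to the Fortnite shipping log, labeled as the TODM Erebus condition.
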(fn Evaluate
  (bind _returnvalue (TouchPlatform|IsNintendoSwitchorNintendoSwitchPreview))
  (Fortnite|FortShippingLog self (Utilities|String|Append "TODM Condition: Erebus = " (Utilities|String|ToString(Boolean) _returnvalue)))
  (return _returnvalue))
