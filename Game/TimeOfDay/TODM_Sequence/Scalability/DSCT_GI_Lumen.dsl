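; A scalability condition for Time of Day Manager sequences that passes when Lumen global illumination is enabled.
; Evaluate queries IsLumenEnabled from the tech art global library, writes the result to the shipping log and returns it.
; Setup binds the condition-changed broadcast to the user settings' global illumination quality change event, so listeners are notified when that setting changes.
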(fn Evaluate
  (bind _isenabled (Class|BTechArtGlobalLibrary|IsLumenEnabled))
  (Fortnite|FortShippingLog self (Utilities|String|Append "TODM Condition: GI_Lumen = " (Utilities|String|ToString(Boolean) _isenabled)))
  (return _isenabled))

(fn SetupOnConditionValueChanged
  (EventDispatchers|BindEventtoOnGlobalIlluminationQualityChanged (Settings|GetFortGameUserSettings) (EventDispatchers|CreateEvent :!event BroadcastOnConditionValueChanged)))
