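; Day sequence condition for time-of-day scalability that reports whether the Medium global illumination tier is active.
; Evaluate returns true when user settings do not allow global illumination, sg.GlobalIlluminationQuality is 1 and PC performance mode is off, and logs the result.
; Setup binds the condition's change broadcast to the Fortnite user settings' GI quality changed event, the only change it listens for.
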(fn Evaluate
  (bind _returnvalue (Class|PlatformHelpersLibrary|IsPcPerfMode))
  (bind _returnvalue_2 (not _returnvalue))
  (bind _returnvalue_3 (Development|GetConsoleVariableIntValue "sg.GlobalIlluminationQuality"))
  (bind _returnvalue_4 (Settings|GetGameUserSettings))
  (Fortnite|FortShippingLog self (Utilities|String|Append "TODM Condition: GI_Medium = " (Utilities|String|ToString(Boolean) (Math|Boolean|ANDBoolean (not (Settings|IsGlobalIlluminationAllowed _returnvalue_4)) (== _returnvalue_3 1) _returnvalue_2))))
  (return (Math|Boolean|ANDBoolean (not (Settings|IsGlobalIlluminationAllowed _returnvalue_4)) (== _returnvalue_3 1) _returnvalue_2)))

(fn SetupOnConditionValueChanged
  (EventDispatchers|BindEventtoOnGlobalIlluminationQualityChanged (Settings|GetFortGameUserSettings) (EventDispatchers|CreateEvent :!event BroadcastOnConditionValueChanged)))
