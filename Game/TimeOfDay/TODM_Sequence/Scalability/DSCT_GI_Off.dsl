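; Day sequence condition for the time-of-day manager that reports whether global illumination is off.
; Evaluate returns true when the sg.GlobalIlluminationQuality console variable is 0 and PC performance mode is not active, and writes the result to the shipping log.
; Setup binds BroadcastOnConditionValueChanged to the game user settings' global illumination quality change event, so changes to that setting broadcast a condition update.
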(fn Evaluate
  (bind _returnvalue (Class|PlatformHelpersLibrary|IsPcPerfMode))
  (bind _returnvalue_2 (not _returnvalue))
  (bind _returnvalue_3 (Development|GetConsoleVariableIntValue "sg.GlobalIlluminationQuality"))
  (Fortnite|FortShippingLog self (Utilities|String|Append "TODM Condition: GI_Off = " (Utilities|String|ToString(Boolean) (and (== _returnvalue_3 0) _returnvalue_2))))
  (return (and (== _returnvalue_3 0) _returnvalue_2)))

(fn SetupOnConditionValueChanged
  (EventDispatchers|BindEventtoOnGlobalIlluminationQualityChanged (Settings|GetFortGameUserSettings) (EventDispatchers|CreateEvent :!event BroadcastOnConditionValueChanged)))
