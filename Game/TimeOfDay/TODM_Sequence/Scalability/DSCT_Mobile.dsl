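; Time of Day Manager (TODM) scalability condition that decides whether the game counts as Mobile.
; Evaluate returns true when either the mobile platform or mobile game check passes, excluding Nintendo Switch (or its preview) and cloud-hosted sessions.
; It writes the result to the Fortnite shipping log before returning it.
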(fn Evaluate
  (bind _returnvalue (Fortnite|IsMobilePlatform))
  (bind _returnvalue_2 (FortGlobals|IsRunninginCloud))
  (bind _returnvalue_3 (TouchPlatform|IsNintendoSwitchorNintendoSwitchPreview))
  (bind _returnvalue_4 (or _returnvalue_3 _returnvalue_2))
  (bind _returnvalue_5 (not _returnvalue_4))
  (bind _returnvalue_6 (TouchPlatform|IsMobileGame))
  (bind _returnvalue_7 (or _returnvalue_6 _returnvalue))
  (bind _returnvalue_8 (and _returnvalue_7 _returnvalue_5))
  (Fortnite|FortShippingLog self (Utilities|String|Append "TODM Condition: Mobile = " (Utilities|String|ToString(Boolean) _returnvalue_8)))
  (return _returnvalue_8))
