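; DSCT_PerfMode is a time-of-day manager scalability condition that reports whether the PC is running in performance mode.
; Its Evaluate function returns the result of PlatformHelpersLibrary IsPcPerfMode.
; It also writes that boolean to the Fortnite shipping log with a TODM Condition prefix.
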(fn Evaluate
  (bind _returnvalue (Class|PlatformHelpersLibrary|IsPcPerfMode))
  (Fortnite|FortShippingLog self (Utilities|String|Append "TODM Condition: PC_PerfMode = " (Utilities|String|ToString(Boolean) _returnvalue)))
  (return _returnvalue))
