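; Time-of-day sequence scalability condition that reports whether volumetric fog should be active, based on the player's graphics settings.
; Evaluate passes when shadow quality in the game user settings is 2 or higher, and logs the result.
; Setup binds the condition's change broadcast to the Fortnite user settings' shadow-quality-changed event, so the condition signals when that setting changes.
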(fn Evaluate
  (bind _returnvalue (Settings|GetGameUserSettings))
  (Fortnite|FortShippingLog self (Utilities|String|Append "TODM Condition: Volumetric Fog = " (Utilities|String|ToString(Boolean) (>= (Settings|GetShadowQuality _returnvalue) 2))))
  (return (>= (Settings|GetShadowQuality _returnvalue) 2)))

(fn SetupOnConditionValueChanged
  (EventDispatchers|BindEventtoOnShadowQualityChanged (Settings|GetFortGameUserSettings) (EventDispatchers|CreateEvent :!event BroadcastOnConditionValueChanged)))
